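# Material function for the FunkyBear Robro backpack that adds an audio-reactive DJ deck display, built from RadioPaca tech art, to incoming material attributes.
# MF_AudioReactivity supplies amplitude and spectrum to MF_FunkyBearBarGraph, which draws three-color bars using rotated, remapped deck UVs.
# A tech art mask input and lerps composite the effects into base color and emissive.
# It also derives a BPM-timed triangle-wave pulse and a timestamp transition with a debug loop switch.
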MaterialFunction /BRCosmetics/Accessories/FORT_Backpacks/Backpack_M_MED_Robro/Skins/FunkyBear/Materials/MF_Robro_Backpack
graph {nodes [
  n1 = Add(A=VectorParameter() {group=RadioPaca Tech Art, parameterName=BaseEmissive, defaultValue=<rgba 0 0 0 1>}.RGB, B=NamedRerouteUsage())
  n2 = LinearInterpolate(A=NamedRerouteUsage(), B=NamedRerouteUsage(), Alpha=NamedRerouteUsage())
  Add(A=LinearInterpolate(A=Constant(), B=Multiply(A=@n1) {constB=2}, Alpha=NamedRerouteUsage()), B=@n2)
  ComponentMask(None=NamedRerouteUsage()) {g=true}
  ComponentMask(None=NamedRerouteUsage()) {g=true}
  ComponentMask(None=NamedRerouteUsage()) {g=true}
  ComponentMask(None=NamedRerouteUsage()) {b=true}
  Constant()
  Constant()
  Constant()
  n3 = GetMaterialAttributes(None=NamedRerouteUsage()) {attributeGetTypes [
    #1
    #2
    #3
    #4
    #5
   ]}
  n4 = SetMaterialAttributes!(MaterialAttributes=@n3.MaterialAttributes, EmissiveColor=NamedRerouteUsage(), BaseColor=NamedRerouteUsage(), Specular, Roughness) {attributeSetTypes [
    #1
    #2
    #3
    #4
   ]}
  FunctionOutput(None=@n4) {outputName=Result_1}
  LinearInterpolate(A=NamedRerouteUsage(), B=LinearInterpolate(A=NamedRerouteUsage(), Alpha=NamedRerouteUsage()) {constB=0}, Alpha=NamedRerouteUsage())
  n5 = NamedRerouteUsage()
  Multiply(A=NamedRerouteUsage(), B=@n5)
  NamedRerouteDeclaration(None=FunctionInput(Preview=TextureSample() {samplerType=SAMPLERTYPE_Masks, texture=/BRCosmetics/Animation/Props/Gadgets_RadioPaca/Textures/T_RadioPaca_Gadgets_FX}.RGB) {bUsePreviewValueAsDefault=true, inputName=In_TechArtMask}) {name=In_TechArtMask, nodeColor=<rgba 0.4531 0 1 1>}
  NamedRerouteDeclaration(None=FunctionInput() {inputName=In_MaterialAttributes, inputType=FunctionInput_MaterialAttributes}) {name=In_MaterialAttributes, nodeColor=<rgba 0 1 0.1328 1>}
  n6 = NamedRerouteUsage()
  n7 = Multiply(A=Time(), B=Divide(A=ScalarParameter() {defaultValue=120, parameterName=Deafault BPM}) {constB=60})
  n8 = ConstantBiasScale(None=Frac(None=NamedRerouteDeclaration(None=@n7) {name=DefaultBPM, nodeColor=<rgba 0.0547 1 0 1>})) {bias=0, scale=2}
  n9 = NamedRerouteDeclaration(None=OneMinus(None=Abs(None=@n8))) {name=DefaultBeat, nodeColor=<rgba 0 1 0.5078 1>}
  n10 = Saturate(None=Multiply(A=@n9, B=@n9))
  NamedRerouteDeclaration(None=Multiply(A=Multiply(A=@n6, B=@n10))) {name=StuffTimedToBeat, nodeColor=<rgba 0.5469 0 1 1>}
  NamedRerouteDeclaration(None=AppendVector(
    A=ScalarParameter() {defaultValue=0.404, group=RadioPaca Tech Art, parameterName=ScaleX_FrontofDJDeck}
    B=ScalarParameter() {defaultValue=0.086, group=RadioPaca Tech Art, parameterName=ScaleY_FrontofDJDeck}
   )) {name=ScaleUVFrontofDJDeck, nodeColor=<rgba 0.5469 0 1 1>}
  NamedRerouteDeclaration(None=VectorParameter() {group=RadioPaca Tech Art, parameterName=LowColor_FrontofDJDeck, defaultValue=<rgba 0 1 1 1>}.RGB) {name=LowColor_Paca, nodeColor=<rgba 1 0.6328 0 1>}
  NamedRerouteDeclaration(None=VectorParameter() {group=RadioPaca Tech Art, parameterName=MidColor_FrontofDJDeck, defaultValue=<rgba 0 1 1 1>}.RGB) {name=MidColor_Paca, nodeColor=<rgba 0.3125 0 1 1>}
  n11 = Divide(A=Time(), B=ScalarParameter() {defaultValue=3, group=FNCSShield_Screen, parameterName=DebugTransitionLength, sortPriority=30})
  n12 = MaterialFunctionCall(
   TimestampA=ScalarParameter() {defaultValue=1, parameterName=Timestamp, sortPriority=30}
   TransitionLength_In=ScalarParameter() {defaultValue=0.1, parameterName=TransitionLength, sortPriority=30}
  ) {materialFunction=&MF_Timestamp_InOut_01}
  NamedRerouteDeclaration(None=StaticSwitchParameter(True=Frac(None=@n11), False=VertexInterpolator(VS=@n12.Linear).PS) {group=FNCSShield_Screen_Graphics, parameterName=DebugModeOn, sortPriority=0}) {name=TimeStamp, nodeColor=<rgba 0.125 1 0 1>}
  NamedRerouteDeclaration(None=@n3.EmissiveColor) {name=In_Emissive, nodeColor=<rgba 1 0.5156 0 1>}
  NamedRerouteDeclaration(None=@n3.BaseColor) {name=In_BaseColor, nodeColor=<rgba 0.1953 0 1 1>}
  NamedRerouteDeclaration(None=@n2) {name=Out_Emissive, nodeColor=<rgba 0.125 1 0 1>}
  n13 = VectorParameter() {group=RadioPaca Tech Art, parameterName=HighColor_FrontofDJDeck, defaultValue=<rgba 0.8469 0.0762 0.8469 1>}
  NamedRerouteDeclaration(None=@n13.RGB) {name=HighColor_Paca, nodeColor=<rgba 0.0078 1 0 1>}
  NamedRerouteDeclaration(None=AppendVector(
    A=ScalarParameter() {defaultValue=0.52, group=RadioPaca Tech Art, parameterName=OffsetX_FrontofDJDeck}
    B=ScalarParameter() {defaultValue=0.904, group=RadioPaca Tech Art, parameterName=OffsetY_FrontofDJDeck}
   )) {name=OffsetUVFrontofDJDeck, nodeColor=<rgba 0 1 0.2266 1>}
  n14 = MaterialFunctionCall() {materialFunction=/Game/Packages/Fortress_SharedMaterials/Base_Material_Functions/BaseCharacterShaderFunctions/Audio/MF_AudioReactivity}
  NamedRerouteDeclaration(None=Multiply(A=ScalarParameter() {defaultValue=1, group=RadioPaca Tech Art, parameterName=AudioReactivityAmplitude}, B=@n14.Amplitude)) {name=Amplitude, nodeColor=<rgba 0 1 0.1094 1>}
  NamedRerouteDeclaration(None=Multiply(A=@n14.Spectrum, B=ScalarParameter() {defaultValue=1, group=RadioPaca Tech Art, parameterName=AudioReactivitySpectrum})) {name=Spectrum, nodeColor=<rgba 1 0 0.75 1>}
  NamedRerouteDeclaration(None=LinearInterpolate(A=NamedRerouteUsage(), B=NamedRerouteUsage(), Alpha=@n5)) {name=Out_BaseColor, nodeColor=<rgba 1 0 0.4453 1>}
  NamedRerouteDeclaration(None=Power(Base=NamedRerouteUsage(), Exp=ScalarParameter() {defaultValue=8, group=RadioPaca Tech Art, parameterName=AmplitudeEmissiveExp}) {constExponent=8}) {name=AmplitudeAdjusted, nodeColor=<rgba 0 1 0.9531 1>}
  n15 = TextureCoordinate()
  n16 = MaterialFunctionCall(Multiplier=Constant() {r=2}) {materialFunction=&Pi}
  n17 = Divide(A=Subtract(A=Rotator(Coordinate=@n15, Time=@n16.Result), B=NamedRerouteUsage()), B=NamedRerouteUsage())
  NamedRerouteDeclaration(None=MaterialFunctionCall(UVs=NamedRerouteDeclaration(None=@n17) {name=ScaledUVs_Deck, nodeColor=<rgba 1 0 0.0938 1>}) {materialFunction=&ScaleUVsByCenter}.0-1 mask) {name=Mask_FrontofDJDeck, nodeColor=<rgba 0 0.4141 1 1>}
  n18 = ScalarParameter() {defaultValue=0.5, parameterName=BoxMaskCenterX}
  n19 = ScalarParameter() {defaultValue=0.5, parameterName=BoxMaskCenterY}
  n20 = ScalarParameter() {defaultValue=0.5, parameterName=BoxMaskBoundsX}
  n21 = ScalarParameter() {defaultValue=0.5, parameterName=BoxMaskBoundsY}
  NamedRerouteDeclaration(None=MaterialFunctionCall!(
    A=TextureCoordinate()
    B=AppendVector(A=@n18, B=@n19)
    Bounds=AppendVector(A=@n20, B=@n21)
    Edge Falloff=ScalarParameter() {defaultValue=0.5, parameterName=BoxMaskEdgeFalloff}
   ) {materialFunction=/Engine/Functions/Engine_MaterialFunctions01/Gradient/BoxMask-2D}.Result) {name=BoxMask, nodeColor=<rgba 0 0.9766 1 1>}
  NamedRerouteDeclaration(None=LinearInterpolate(A=TextureCoordinate(), B=NamedRerouteUsage(), Alpha=NamedRerouteUsage())) {name=FinalLerpedUVs, nodeColor=<rgba 0 0.5781 1 1>}
  NamedRerouteDeclaration(None=LinearInterpolate(A=Constant2Vector(), B=NamedRerouteUsage(), Alpha=NamedRerouteUsage())) {name=FinalLerpedGraphBands, nodeColor=<rgba 1 0.8315 0 1>}
  NamedRerouteDeclaration(None=MaterialFunctionCall(X=Constant() {r=6}, Y=Constant() {r=20}) {materialFunction=&MakeFloat2}.Result) {name=Bands_FrontOfDJDeck, nodeColor=<rgba 0.7812 0 1 1>}
  NamedRerouteDeclaration(None=MaterialFunctionCall!(
    In_UV=NamedRerouteUsage()
    In_JustBars=NamedRerouteUsage()
    In_Bands=NamedRerouteUsage()
    In_LowColor=NamedRerouteUsage()
    In_MidColor=NamedRerouteUsage()
    In_HighColor=NamedRerouteUsage()
    In_Spectrum=NamedRerouteUsage()
    In_Amplitude=NamedRerouteUsage()
    In_TimeStamp=NamedRerouteUsage()
   ) {materialFunction=&/MF_FunkyBearBarGraph}.MaskedColor) {name=FinalMaskedColorBands, nodeColor=<rgba 0.9688 0 1 1>}
  NamedRerouteDeclaration(None=Constant() {r=1}) {name=JustBars, nodeColor=<rgba 0 0.5781 1 1>}
  NamedRerouteUsage()
 ]}
